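; Telegraph actor that signals where the medium TidalCrane will unburrow, using a Niagara effect and moving audio.
; SetUnborrowLocation stores the target, whose RepNotify sets the effect's EndPosition and activates it.
; It also starts a short-lived looping timer that steps a lerp alpha, sliding the audio from the instigator's current location toward the target until a second timer clears it.
; Effect and audio updates run only where cosmetic events can execute.
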(event EventBeginPlay)

(event Collision|EventActorBeginOverlap
  (params (OtherActor :type Actor%20Object%20Reference)))

(event EventTick
  (params (DeltaSeconds :type Float%20%28single-precision%29)))

(event Custom|SetUnborrowLocation
  (params (UnburrowLocation :type Vector))
  (Variables|Default|SetUnburrowLocation UnburrowLocation)
  (bind _returnvalue (Utilities|Time|SetTimerbyEvent (EventDispatchers|CreateEvent :!event Custom|UpdateUnburrowTellAudioLocation) 0.1 true))
  (Variables|Default|SetAudioLocationUpdateTimer _returnvalue)
  (Utilities|Time|SetTimerbyEvent (delegate Custom|ClearTellTimers) 1.5))

(event Custom|ClearTellTimers
  (Utilities|Time|ClearandInvalidateTimerbyHandle (Variables|Default|GetAudioLocationUpdateTimer)))

(event Custom|UpdateUnburrowTellAudioLocation
  (bind _unburrowtellaudio (Variables|Default|GetUnburrowTellAudio))
  (Networking|CanExecuteCosmeticEvents
    (:True
      (Variables|Default|SetAudioLocationAlpha (+ (Variables|Default|GetAudioLocationAlpha) 0.1))
      (Utilities|IsValid _unburrowtellaudio
        (:Is%20Valid
          (Transformation|SetWorldLocation _unburrowtellaudio (Math|Vector|Lerp(Vector) (Transformation|GetActorLocation (Variables|Actor|GetInstigator)) (Variables|Default|GetUnburrowLocation) (Variables|Default|GetAudioLocationAlpha)) false false))))))

(fn OnRep_UnburrowLocation
  (bind _ns_bomber_burrow_telegraph (Variables|Default|GetNS_Bomber_Burrow_Telegraph))
  (Networking|CanExecuteCosmeticEvents
    (:True
      (Niagara|SetNiagaraVariable(Position) _ns_bomber_burrow_telegraph "EndPosition" (Variables|Default|GetUnburrowLocation))
      (Components|Activation|SetActive _ns_bomber_burrow_telegraph true false))))

(fn ConstructionScript)
